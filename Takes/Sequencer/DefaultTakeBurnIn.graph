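# Level Sequence burn-in widget for Takes, overlaying six text slots in top and bottom bars around an optional tinted watermark.
# SetSettings applies font and watermark from DefaultBurnInOptions, whose format strings the text getters fill through PerformTextFormat.
# CacheData derives timecode, frames, shot names, cine camera data, and slate and take number from the active shot's TakeMetaData.
# A bCached flag, cleared each Tick, limits that work to once per frame.
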WidgetBlueprint /Takes/Sequencer/DefaultTakeBurnIn
graphs [
 CacheData {nodes [
   CacheData() {is=title K2Node_FunctionEntry}
   > n1 = Branch(Condition=Get bCached() -> (bCached:bool))
   Set bCached(execute=@n1.else, bCached:bool=true) -> (Output_Get:bool)
   n2 = Break Level Sequence Player Snapshot!(LevelSequencePlayerSnapshot:"Level Sequence Player Snapshot Structure (by ref)"=Get FrameInformation() -> (FrameInformation:Level Sequence Player Snapshot Structure)) -> (
    RootName:String
    RootTime:Qualified Frame Time Structure
    SourceTime:Qualified Frame Time Structure
    CurrentShotName:String
    CurrentShotLocalTime:Qualified Frame Time Structure
    CurrentShotSourceTime:Qualified Frame Time Structure
    SourceTimecode:String
    CameraComponent:Camera Component Soft Object Reference
    ActiveShot:Level Sequence Object Reference
   ) {is=title K2Node_BreakStruct}
   n3 = float / float(A=QualifiedFrameTime To Seconds!(InFrameTime:"Qualified Frame Time Structure (by ref)"=@n2.RootTime) -> (ReturnValue:float), B=3600) {is=title K2Node_CallFunction}
   > Set hh(hh:Text="To Text (Integer)"(Value=Floor(A=@n3) {is=title K2Node_CallFunction}, MinimumIntegralDigits=2)) -> (Output_Get:Text)
   n4 = float / float(A=QualifiedFrameTime To Seconds(InFrameTime=@n2.RootTime), B=60)
   > Set mm(mm:Text="To Text (Integer)"(Value=Floor(A=@n4), MinimumIntegralDigits=2)) -> (Output_Get:Text)
   n5 = "% (Float)"!(A:double=QualifiedFrameTime To Seconds(InFrameTime=@n2.RootTime), B:double=60) -> (ReturnValue:double)
   > Set ss(ss:Text="To Text (Integer)"(Value=Floor(A=@n5), MinimumIntegralDigits=2)) -> (Output_Get:Text)
   n6 = BreakQualifiedFrameTime!(InFrameTime:"Qualified Frame Time Structure (by ref)"=@n2.RootTime) -> (Frame:Frame Number Structure, FrameRate:Frame Rate Structure, SubFrame:float)
   n7 = float * float(
    A=Fraction!(A:double=@n5) -> (ReturnValue:double)
    B=FrameRate To Interval!(InFrameRate:Frame Rate Structure=@n6.FrameRate) -> (ReturnValue:float)
   )
   > Set ff(ff:Text="To Text (Integer)"(Value=Round(A=@n7) {is=title K2Node_CallFunction}, bUseGrouping=false, MinimumIntegralDigits=2)) -> (Output_Get:Text)
   n8 = float * float(A=QualifiedFrameTime To Seconds(InFrameTime=@n2.RootTime), B=FrameRate To Interval(InFrameRate=@n6.FrameRate))
   > Set MasterFrame(MasterFrame:Text="To Text (Integer)"(Value=Round(A=@n8), bUseGrouping=false, MinimumIntegralDigits="To Integer (Byte)"!(InByte:Byte=0) -> (ReturnValue:int=0))) -> (Output_Get:Text)
   n9 = float * float(
    A=QualifiedFrameTime To Seconds(InFrameTime=@n2.CurrentShotLocalTime)
    B=FrameRate To Interval(InFrameRate=BreakQualifiedFrameTime(InFrameTime=@n2.CurrentShotLocalTime).FrameRate)
   )
   > Set ShotFrame(ShotFrame:Text="To Text (Integer)"(Value=Round(A=@n9), bUseGrouping=false, MinimumIntegralDigits="To Integer (Byte)"(InByte=0) -> (ReturnValue=0))) -> (Output_Get:Text)
   n10 = Break Level Sequence Player Snapshot(LevelSequencePlayerSnapshot=Get FrameInformation())
   > Set MasterName(MasterName:Text="To Text (String)"(InString=@n10)) -> (Output_Get:Text)
   > Set ShotName(ShotName:Text="To Text (String)"(InString=@n10.CurrentShotName)) -> (Output_Get:Text)
   > Set SourceTimecode(SourceTimecode:Text="To Text (String)"(InString=@n10.SourceTimecode)) -> (Output_Get:Text)
   > n11 = Sequence()
   n12 = Resolve Soft Reference!(Input:Camera Component Soft Object Reference=@n10.CameraComponent) -> (Output:Camera Component Object Reference)
   > Branch(Condition=IsValid(Object=@n12))
   > n13 = Cast To CineCameraComponent!(execute, Object=@n12) -> (then, CastFailed, AsCine Camera Component:Cine Camera Component Object Reference) {is=title K2Node_DynamicCast}
   > Set FocalLength(FocalLength:Text="To Text (Float)"(
     Value=Get CurrentFocalLength!(self:Cine Camera Component Object Reference=@n13) -> (CurrentFocalLength:float)
     MinimumIntegralDigits=1
     MaximumIntegralDigits=324
     MinimumFractionalDigits=0
     MaximumFractionalDigits=2
    )) -> (Output_Get:Text)
   > Set FocusDistance(FocusDistance:Text="To Text (Float)"(
     Value=Get CurrentFocusDistance!(self:Cine Camera Component Object Reference=@n13) -> (CurrentFocusDistance:float)
     MinimumIntegralDigits=1
     MaximumIntegralDigits=324
     MinimumFractionalDigits=0
     MaximumFractionalDigits=2
    )) -> (Output_Get:Text)
   > Set Aperture(Aperture:Text="To Text (Float)"(
     Value=Get CurrentAperture!(self:Cine Camera Component Object Reference=@n13) -> (CurrentAperture:float)
     MinimumIntegralDigits=1
     MaximumIntegralDigits=324
     MinimumFractionalDigits=0
     MaximumFractionalDigits=2
    )) -> (Output_Get:Text)
   n14 = Break Camera Filmback Settings!(CameraFilmbackSettings:"Camera Filmback Settings Structure (by ref)"=Get Filmback!(self:Cine Camera Component Object Reference=@n13) -> (Filmback:Camera Filmback Settings Structure)) -> (SensorWidth:float=~, SensorHeight:float=~, SensorHorizontalOffset:float=~, SensorVerticalOffset:float=~, SensorAspectRatio:float=~) {is=title K2Node_BreakStruct}
   > Set SensorWidth(SensorWidth:Text="To Text (Float)"(Value=@n14, MinimumIntegralDigits=1, MaximumIntegralDigits=324, MinimumFractionalDigits=0, MaximumFractionalDigits=2)) -> (Output_Get:Text)
   > Set SensorHeight(SensorHeight:Text="To Text (Float)"(Value=@n14.SensorHeight, MinimumIntegralDigits=1, MaximumIntegralDigits=324, MinimumFractionalDigits=0, MaximumFractionalDigits=2)) -> (Output_Get:Text)
   > Set SensorAspectRatio(SensorAspectRatio:Text="To Text (Float)"(Value=@n14.SensorAspectRatio, MinimumIntegralDigits=1, MaximumIntegralDigits=324, MinimumFractionalDigits=0, MaximumFractionalDigits=2)) -> (Output_Get:Text)
   > Set Translation(Translation:Vector=Get World Location!(self:Scene Component Object Reference=@n13) -> (ReturnValue:Vector) {is=title K2Node_CallFunction}) -> (Output_Get:Vector)
   > Set Rotation(Rotation:Rotator=Get World Rotation!(self:Scene Component Object Reference=@n13) -> (ReturnValue:Rotator) {is=title K2Node_CallFunction}) -> (Output_Get:Rotator)
   Branch(execute=@n11.then_1, Condition=IsValid(Object=@n10.ActiveShot))
   > n15 = Cast To TakeMetaData!(
    execute
    Object=FindMetaDataByClass!(self:Level Sequence Object Reference=@n10.ActiveShot, InClass:Object Class Reference=/Script/TakesCore.TakeMetaData) -> (ReturnValue:Object Reference)
   ) -> (then, CastFailed, AsTake Meta Data:Take Meta Data Object Reference) {is=title K2Node_DynamicCast}
   > Set Slate(Slate:Text="To Text (String)"(InString=GetSlate!(self:Take Meta Data Object Reference=@n15) -> (ReturnValue:String))) -> (Output_Get:Text)
   > Set TakeNumber(TakeNumber:Text="To Text (Integer)"(Value=GetTakeNumber!(self:Take Meta Data Object Reference=@n15) -> (ReturnValue:int=0))) -> (Output_Get:Text)
  ]}
 EventGraph {nodes [
   Event Construct()
   n1 = UTC Now!() -> (ReturnValue:Date Time Structure)
   > Set Date(Date:Text=As Date!(InDateTime:"Date Time Structure (by ref)"=@n1, InDateStyle:EDateTimeStyle Enum=Default) -> (ReturnValue:Text)) -> (Output_Get:Text)
   > Set EngineVersion(EngineVersion:Text="To Text (String)"(InString=GetEngineVersion!() -> (ReturnValue:String))) -> (Output_Get:Text)
   > Set Time(Time:Text="As Time (from UTC)"!(InDateTime:"Date Time Structure (by ref)"=@n1, InTimeZone:String, InTimeStyle:EDateTimeStyle Enum=Default) -> (ReturnValue:Text)) -> (Output_Get:Text)
   n2 = Event SetSettings!() -> (OutputDelegate, then, InSettings:Object Reference) {is=title K2Node_Event}
   > n3 = Cast To DefaultBurnInOptions!(execute, Object=@n2.InSettings) -> (then, CastFailed, AsDefault Burn in Options:Default Burn In Options Object Reference) {is=title K2Node_DynamicCast}
   > Set Options(Options:Default Burn In Options Object Reference=@n3) -> (Output_Get:Default Burn In Options Object Reference)
   n4 = Get Options() -> (Options:Default Burn In Options Object Reference)
   > SetFont(
    self=[
     Get BottomLeft() -> (BottomLeft:Text Object Reference)
     Get TopLeft() -> (TopLeft:Text Object Reference)
     Get TopCenter() -> (TopCenter:Text Object Reference)
     Get TopRight() -> (TopRight:Text Object Reference)
     Get BottomRight() -> (BottomRight:Text Object Reference)
     Get BottomCenter() -> (BottomCenter:Text Object Reference)
    ]
    InFontInfo=Get Font!(self:Default Burn In Options Object Reference=@n4) -> (Font:Slate Font Info Structure)
   )
   n5 = Get Watermark!(self:Default Burn In Options Object Reference=@n4) -> (Watermark:Texture 2D Object Reference)
   > n6 = Branch(Condition=IsValid(Object=@n5))
   n7 = Get Watermark!() -> (Watermark:Image Object Reference)
   > SetVisibility(self=@n7, InVisibility=Visible)
   n8 = Get Watermark()
   > SetColorAndOpacity(
    self:Image Object Reference=@n8
    InColorAndOpacity:Linear Color Structure=Get WatermarkTint!(self:Default Burn In Options Object Reference=Get Options()) -> (WatermarkTint:Linear Color Structure)
   )
   > SetBrushFromTexture!(execute, self:Image Object Reference=@n8, Texture:Texture 2D Object Reference=@n5, bMatchSize:bool=false) -> (then)
   SetVisibility(execute=@n6.else, self=@n7, InVisibility=Hidden)
   Event Tick()
   > Set bCached(bCached=false)
  ]}
 GetSettingsClass {nodes [
   GetSettingsClass() {is=title K2Node_FunctionEntry}
   > Return Node!(execute, ReturnValue:Level Sequence Burn In Init Settings Class Reference=/Engine/Sequencer/DefaultBurnInOptions.DefaultBurnInOptions_C)
  ]}
 Get_BottomCenter_Text_0 {nodes [
   Get_BottomCenter_Text_0() {is=title K2Node_FunctionEntry}
   > CacheData() {is=title K2Node_CallFunction}
   n1 = Get Options()
   > Return Node(ReturnValue:Text=Perform Text Format!(Format:Text=Get BottomCenterText!(self:Default Burn In Options Object Reference=@n1) -> (BottomCenterText:Text)) -> (FormattedText:Text) {is=title K2Node_MacroInstance})
  ]}
 Get_BottomLeft_Text_0 {nodes [
   Get_BottomLeft_Text_0() {is=title K2Node_FunctionEntry}
   > CacheData()
   n1 = Get Options()
   > Return Node(ReturnValue=Perform Text Format(Format=Get BottomLeftText!(self:Default Burn In Options Object Reference=@n1) -> (BottomLeftText:Text)))
  ]}
 Get_BottomRight_Text_0 {nodes [
   Get_BottomRight_Text_0() {is=title K2Node_FunctionEntry}
   > CacheData()
   n1 = Get Options()
   > Return Node(ReturnValue=Perform Text Format(Format=Get BottomRightText!(self:Default Burn In Options Object Reference=@n1) -> (BottomRightText:Text)))
  ]}
 Get_TopCenter_Text_0 {nodes [
   Get_TopCenter_Text_0() {is=title K2Node_FunctionEntry}
   > CacheData()
   n1 = Get Options()
   > Return Node(ReturnValue=Perform Text Format(Format=Get TopCenterText!(self:Default Burn In Options Object Reference=@n1) -> (TopCenterText:Text)))
  ]}
 Get_TopLeft_Text_0 {nodes [
   Get_TopLeft_Text_0() {is=title K2Node_FunctionEntry}
   > CacheData()
   n1 = Get Options()
   > Return Node(ReturnValue=Perform Text Format(Format=Get TopLeftText!(self:Default Burn In Options Object Reference=@n1) -> (TopLeftText:Text)))
  ]}
 Get_TopRight_Text_0 {nodes [
   Get_TopRight_Text_0() {is=title K2Node_FunctionEntry}
   > CacheData()
   n1 = Get Options()
   > Return Node(ReturnValue=Perform Text Format(Format=Get TopRightText!(self:Default Burn In Options Object Reference=@n1) -> (TopRightText:Text)))
  ]}
 PerformTextFormat {nodes [
   n1 = BreakRotator(InRot=Get Rotation() -> (Rotation:Rotator))
   n2 = BreakVector(InVec=Get Translation() -> (Translation:Vector))
   Outputs!(FormattedText:Text=Format Text!(
     Format=Inputs!() -> (Format:Text)
     hh:Text=Get hh() -> (hh:Text)
     mm:Text=Get mm() -> (mm:Text)
     ss:Text=Get ss() -> (ss:Text)
     ff:Text=Get ff() -> (ff:Text)
     MasterFrame:Text=Get MasterFrame() -> (MasterFrame:Text)
     ShotFrame:Text=Get ShotFrame() -> (ShotFrame:Text)
     SourceTimecode:Text=Get SourceTimecode() -> (SourceTimecode:Text)
     MasterName:Text=Get MasterName() -> (MasterName:Text)
     ShotName:Text=Get ShotName() -> (ShotName:Text)
     Date:Text=Get Date() -> (Date:Text)
     FocalLength:Text=Get FocalLength() -> (FocalLength:Text)
     FocusDistance:Text=Get FocusDistance() -> (FocusDistance:Text)
     Aperture:Text=Get Aperture() -> (Aperture:Text)
     SensorWidth:Text=Get SensorWidth() -> (SensorWidth:Text)
     SensorHeight:Text=Get SensorHeight() -> (SensorHeight:Text)
     SensorAspectRatio:Text=Get SensorAspectRatio() -> (SensorAspectRatio:Text)
     EngineVersion:Text=Get EngineVersion() -> (EngineVersion:Text)
     TranslationX:Text="To Text (Float)"(Value=@n2, MinimumIntegralDigits=1, MaximumIntegralDigits=324, MinimumFractionalDigits=0, MaximumFractionalDigits=2)
     TranslationY:Text="To Text (Float)"(Value=@n2.Y, MinimumIntegralDigits=1, MaximumIntegralDigits=324, MinimumFractionalDigits=0, MaximumFractionalDigits=2)
     TranslationZ:Text="To Text (Float)"(Value=@n2.Z, MinimumIntegralDigits=1, MaximumIntegralDigits=324, MinimumFractionalDigits=0, MaximumFractionalDigits=2)
     RotationX:Text="To Text (Float)"(Value=@n1, MinimumIntegralDigits=1, MaximumIntegralDigits=324, MinimumFractionalDigits=0, MaximumFractionalDigits=2)
     RotationY:Text="To Text (Float)"(Value=@n1.Pitch, MinimumIntegralDigits=1, MaximumIntegralDigits=324, MinimumFractionalDigits=0, MaximumFractionalDigits=2)
     RotationZ:Text="To Text (Float)"(Value=@n1.Yaw, MinimumIntegralDigits=1, MaximumIntegralDigits=324, MinimumFractionalDigits=0, MaximumFractionalDigits=2)
     Slate:Text=Get Slate() -> (Slate:Text)
     TakeNumber:Text=Get TakeNumber() -> (TakeNumber:Text)
     Time:Text=Get Time() -> (Time:Text)
    ) -> (Result))
  ]}
]
variables [
 Aperture {type=string, default=""}
 Background Color {type=LinearColor, default=<rgba 0 0 0 0.333>}
 bCached {type=boolean}
 Date {type=string, default=""}
 EngineVersion {type=string, default=""}
 ff {type=string, default=""}
 FocalLength {type=string, default=""}
 FocusDistance {type=string, default=""}
 Foreground Color {type=LinearColor, default=<rgba 1 1 1 0.5>}
 hh {type=string, default=""}
 MasterFrame {type=string, default=""}
 MasterName {type=string, default=""}
 mm {type=string, default=""}
 Options {type=DefaultBurnInOptions_C, default=null}
 Rotation {type=Rotator}
 SensorAspectRatio {type=string, default=""}
 SensorHeight {type=string, default=""}
 SensorWidth {type=string, default=""}
 ShotFrame {type=string, default=""}
 ShotName {type=string, default=""}
 Slate {type=string, default=""}
 SourceTimecode {type=string, default=""}
 ss {type=string, default=""}
 TakeNumber {type=string, default=""}
 Time {type=string, default=""}
 Translation {type=Vector}
]
widgets {
 info {inheritedWidgetCount=0, parentClass=/Script/LevelSequence.LevelSequenceBurnIn, rootWidgetClass=&CanvasPanel, widgetCount=15}
 widgets [
  CanvasPanel_0 {widgetClassPath=&CanvasPanel, widget=CanvasPanel_0}
  VerticalBox_16 {
   widgetClassPath=&VerticalBox
   parent=CanvasPanel_0
   widget=VerticalBox_16
   slot {class=CanvasPanelSlot, properties {layoutData {alignment=<0 0>, anchors {maximum=<1 1>, minimum=<0 0>}, offsets=<ltrb 0 0 0 0>}}}
  }
  Border_0 {
   widgetClassPath=&Border
   parent=VerticalBox_16
   widget=Border_0
   properties {brushColor=<rgba 0 0 0 1>, padding=<ltrb 2 2 2 2>}
   slot {class=VerticalBoxSlot, properties {verticalAlignment=VAlign_Top}}
  }
  SizeBox_0 {
   widgetClassPath=&SizeBox
   parent=Border_0
   widget=SizeBox_0
   properties {heightOverride=20}
   slot {class=BorderSlot, properties {padding=<ltrb 2 2 2 2>}}
  }
  HorizontalBox_2 {widgetClassPath=&HorizontalBox, parent=SizeBox_0, widget=HorizontalBox_2, slot {class=SizeBoxSlot}}
  TopLeft {
   widgetClassPath=&TextBlock
   bIsVariable=true
   parent=HorizontalBox_2
   widget=TopLeft
   properties {font {
     bForceMonospaced=false
     bMaterialIsStencil=false
     fontMaterial=null
     fontObject=/Engine/EngineFonts/DroidSansMono
     letterSpacing=0
     monospacedWidth=1
     size=18
     skewAmount=0
     typefaceFontName=Font
     outlineSettings {
      bApplyOutlineToDropShadows=false
      bMiteredCorners=false
      bSeparateFillAlpha=false
      outlineMaterial=null
      outlineSize=0
      outlineColor=<rgba 0 0 0 1>
     }
    }}
   slot {class=HorizontalBoxSlot, properties {size {sizeRule=Fill, value=1}}}
  }
  TopCenter {
   widgetClassPath=&TextBlock
   bIsVariable=true
   parent=HorizontalBox_2
   widget=TopCenter
   properties {font {
     bForceMonospaced=false
     bMaterialIsStencil=false
     fontMaterial=null
     fontObject=&DroidSansMono
     letterSpacing=0
     monospacedWidth=1
     size=18
     skewAmount=0
     typefaceFontName=Font
     outlineSettings {
      bApplyOutlineToDropShadows=false
      bMiteredCorners=false
      bSeparateFillAlpha=false
      outlineMaterial=null
      outlineSize=0
      outlineColor=<rgba 0 0 0 1>
     }
    }}
   slot {class=HorizontalBoxSlot, properties {horizontalAlignment=HAlign_Center}}
  }
  TopRight {
   widgetClassPath=&TextBlock
   bIsVariable=true
   parent=HorizontalBox_2
   widget=TopRight
   properties {font {
     bForceMonospaced=false
     bMaterialIsStencil=false
     fontMaterial=null
     fontObject=&DroidSansMono
     letterSpacing=0
     monospacedWidth=1
     size=18
     skewAmount=0
     typefaceFontName=Font
     outlineSettings {
      bApplyOutlineToDropShadows=false
      bMiteredCorners=false
      bSeparateFillAlpha=false
      outlineMaterial=null
      outlineSize=0
      outlineColor=<rgba 0 0 0 1>
     }
    }}
   slot {class=HorizontalBoxSlot, properties {horizontalAlignment=HAlign_Right, size {sizeRule=Fill, value=1}}}
  }
  Watermark {
   widgetClassPath=&Image
   bIsVariable=true
   parent=VerticalBox_16
   widget=Watermark
   properties {visibility=Hidden}
   slot {class=VerticalBoxSlot, properties {size {sizeRule=Fill, value=1}}}
  }
  Border_3 {
   widgetClassPath=&Border
   parent=VerticalBox_16
   widget=Border_3
   properties {brushColor=<rgba 0 0 0 1>, padding=<ltrb 2 2 2 2>}
   slot {class=VerticalBoxSlot, properties {verticalAlignment=VAlign_Top}}
  }
  SizeBox_1 {
   widgetClassPath=&SizeBox
   parent=Border_3
   widget=SizeBox_1
   properties {heightOverride=20}
   slot {class=BorderSlot, properties {padding=<ltrb 2 2 2 2>}}
  }
  HorizontalBox_29 {widgetClassPath=&HorizontalBox, parent=SizeBox_1, widget=HorizontalBox_29, slot {class=SizeBoxSlot}}
  BottomLeft {
   widgetClassPath=&TextBlock
   bIsVariable=true
   parent=HorizontalBox_29
   widget=BottomLeft
   properties {font {
     bForceMonospaced=false
     bMaterialIsStencil=false
     fontMaterial=null
     fontObject=&DroidSansMono
     letterSpacing=0
     monospacedWidth=1
     size=18
     skewAmount=0
     typefaceFontName=Font
     outlineSettings {
      bApplyOutlineToDropShadows=false
      bMiteredCorners=false
      bSeparateFillAlpha=false
      outlineMaterial=null
      outlineSize=0
      outlineColor=<rgba 0 0 0 1>
     }
    }}
   slot {class=HorizontalBoxSlot, properties {size {sizeRule=Fill, value=1}}}
  }
  BottomCenter {
   widgetClassPath=&TextBlock
   bIsVariable=true
   parent=HorizontalBox_29
   widget=BottomCenter
   properties {font {
     bForceMonospaced=false
     bMaterialIsStencil=false
     fontMaterial=null
     fontObject=&DroidSansMono
     letterSpacing=0
     monospacedWidth=1
     size=18
     skewAmount=0
     typefaceFontName=Font
     outlineSettings {
      bApplyOutlineToDropShadows=false
      bMiteredCorners=false
      bSeparateFillAlpha=false
      outlineMaterial=null
      outlineSize=0
      outlineColor=<rgba 0 0 0 1>
     }
    }}
   slot {class=HorizontalBoxSlot, properties {horizontalAlignment=HAlign_Center}}
  }
  BottomRight {
   widgetClassPath=&TextBlock
   bIsVariable=true
   parent=HorizontalBox_29
   widget=BottomRight
   properties {font {
     bForceMonospaced=false
     bMaterialIsStencil=false
     fontMaterial=null
     fontObject=&DroidSansMono
     letterSpacing=0
     monospacedWidth=1
     size=18
     skewAmount=0
     typefaceFontName=Font
     outlineSettings {
      bApplyOutlineToDropShadows=false
      bMiteredCorners=false
      bSeparateFillAlpha=false
      outlineMaterial=null
      outlineSize=0
      outlineColor=<rgba 0 0 0 1>
     }
    }}
   slot {class=HorizontalBoxSlot, properties {horizontalAlignment=HAlign_Right, size {sizeRule=Fill, value=1}}}
  }
 ]
}
